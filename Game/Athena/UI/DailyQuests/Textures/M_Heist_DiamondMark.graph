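# Translucent UI material that draws the Heist diamond mark as a white, anti-aliased shape, stored with the Athena daily quest textures.
# UVs are scaled about their center by the red channel of the Scale parameter, used for both axes.
# The red channel of the diamond texture feeds AAThreshold as a gradient, with Edge setting the cutoff and Soften overriding the filter width.
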Material /Game/Athena/UI/DailyQuests/Textures/M_Heist_DiamondMark
graph {
 properties {
  bUsedWithStaticMesh=false
  blendMode=BLEND_Translucent
  materialDomain=MD_UI
  pixelDepthOffsetMode=PDOM_Legacy
  usedShadingModels=Default Lit
 }
 nodes [
  Constant3Vector() {constant=<rgba 1 1 1 0>}
  n1 = TextureCoordinate()
  n2 = VectorParameter() {parameterName=Scale, defaultValue=<rgba 1 1 0 1>}
  n3 = AppendVector(A=@n2.R, B=@n2.R)
  MaterialFunctionCall!(
   Threshold=ScalarParameter() {defaultValue=0.8, parameterName=Edge}
   Gradient=TextureSample(UVs=MaterialFunctionCall(UVs=@n1, Texture Scale=@n3) {materialFunction=&ScaleUVsByCenter}.UVs) {samplerType=SAMPLERTYPE_LinearColor, texture=&/T_UI_Heist_DiamondMark}.R
   Filter Width Override=ScalarParameter() {parameterName=Soften}
   Filter Width Fudge Factor
  ) {materialFunction=/Game/UI/Materials/AAThreshold}
 ]
}
